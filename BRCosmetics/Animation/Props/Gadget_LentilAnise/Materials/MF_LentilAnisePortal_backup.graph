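# Material function, named as a backup, that layers a swirling green portal effect onto input material attributes for the LentilAnise gadget prop.
# Four channels of one mask texture rotate at separate speeds over scaled UVs, with three-colour gradients providing tints.
# Radial falloffs distorted by Fresnel build a glowing core, alongside sphere masks and a separate Fresnel rim.
# Named reroutes carry these parts towards emissive, opacity and Z-axis world position offset outputs.
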MaterialFunction /BRCosmetics/Animation/Props/Gadget_LentilAnise/Materials/MF_LentilAnisePortal_backup
graph {nodes [
  Constant()
  Constant()
  n1 = GetMaterialAttributes(None=NamedRerouteUsage()) {attributeGetTypes [#1]}
  n2 = NamedRerouteUsage()
  n3 = SetMaterialAttributes!(
   MaterialAttributes=@n1.MaterialAttributes
   BaseColor=VectorParameter() {group=Wrap_Ruckus, parameterName=Ruck_BaseColor-----, sortPriority=0}.RGB
   Roughness=ScalarParameter() {defaultValue=0.1, group=Wrap_Ruckus, parameterName=Ruck_Roughness, sortPriority=0}
   Specular=ScalarParameter() {defaultValue=0.5, group=Wrap_Ruckus, parameterName=Ruck_Specular, sortPriority=0}
   EmissiveColor=Multiply(A=NamedRerouteUsage()) {constB=2}
   Opacity=@n2
   WorldPositionOffset=NamedRerouteUsage()
   OpacityMask=@n2
  ) {attributeSetTypes [
    #2
    #3
    #4
    #5
    #6
    #7
    #8
   ]}
  FunctionOutput(None=@n3)
  n4 = VectorParameter() {group=Wrap_Ruckus, parameterName=Ruck_Core02_CoreGlow------, defaultValue=<rgba 0.2811 0.6562 0.2738 0>}
  n5 = Fresnel(ExponentIn=ScalarParameter() {defaultValue=0.4967, group=Wrap_Ruckus, parameterName=Ruck_Core00_DistortionExp-----}) {baseReflectFraction=0}
  n6 = ScalarParameter() {defaultValue=0.27, group=Wrap_Ruckus, parameterName=Ruck_Core00_DistortionNormal}
  n7 = Multiply(A=Distance(A=TextureCoordinate(), B=Constant2Vector() {g=0.5, r=0.5}), B=OneMinus(None=Multiply(A=@n5, B=@n6)))
  n8 = Divide(A=@n7, B=ScalarParameter() {defaultValue=1, group=Wrap_Ruckus, parameterName=Ruck_Core02_DistanceFalloff})
  n9 = Clamp(None=Power(Base=OneMinus(None=@n8), Exp=ScalarParameter() {defaultValue=18, group=Wrap_Ruckus, parameterName=Ruck_Core02_Fresnel_Exp}))
  Multiply(A=Multiply(A=@n4.RGB, B=ScalarParameter() {defaultValue=0.45, group=Wrap_Ruckus, parameterName=Ruck_Core02_Glow}), B=@n9)
  n10 = Divide(A=@n7, B=ScalarParameter() {defaultValue=1.3, group=Wrap_Ruckus, parameterName=Ruck_Core03_DistanceFalloff})
  n11 = Clamp(None=Power(Base=OneMinus(None=@n10), Exp=ScalarParameter() {defaultValue=3, group=Wrap_Ruckus, parameterName=Ruck_Core03_Fresnel_Exp}))
  Multiply(A=VectorParameter() {group=Wrap_Ruckus, parameterName=Ruck_Core03_CoreGlow-----, defaultValue=<rgba 0.3039 0.7529 0.0823 1>}.RGB, B=@n11)
  Multiply() {constA=-1, constB=2}
  n12 = TextureCoordinate()
  n13 = Constant2Vector() {g=0.5, r=0.5}
  n14 = Multiply() {constA=0.333}
  n15 = Constant() {r=0.2}
  NamedRerouteDeclaration(None=Power(Base=SphereMask(A=@n12, B=@n13, Radius=@n14, Hardness=@n15) {attenuationRadius=1.2}) {constExponent=1}) {name=SphereMaskDistort, nodeColor=<rgba 0 1 0.6953 1>}
  NamedRerouteDeclaration(None=NamedRerouteUsage()) {name=EmissiveCore, nodeColor=<rgba 0.6641 1 0 1>}
  n16 = LinearInterpolate(A=NamedRerouteUsage(), B=NamedRerouteUsage(), Alpha=NamedRerouteUsage())
  n17 = NamedRerouteUsage()
  n18 = NamedRerouteUsage()
  NamedRerouteDeclaration(None=LinearInterpolate(
    A=LinearInterpolate(A=@n16, B=@n17, Alpha=@n18)
    B=VectorParameter() {group=Wrap_Ruckus, parameterName=Ruck_Twirl_Channel_01_Color, defaultValue=<rgba 1 1 1 1>}.RGB
    Alpha=NamedRerouteUsage()
   )) {name=RotatingTextures, nodeColor=<rgba 1 0 0.9844 1>}
  NamedRerouteDeclaration(None=SphereMask(
    A=TextureCoordinate()
    B=Constant2Vector() {g=0.5, r=0.5}
    Radius=ScalarParameter() {defaultValue=0.33, group=Wrap_Ruckus, parameterName=Ruck_Core_SphereMask_Radius}
    Hardness=ScalarParameter() {defaultValue=0.1301, group=Wrap_Ruckus, parameterName=Ruck_Core_SphereMask_Hardness-----}
   ) {attenuationRadius=1.2}) {name=SphereMask, nodeColor=<rgba 0 1 0.6953 1>}
  NamedRerouteDeclaration(None=MaterialFunctionCall(UVs=TextureCoordinate(), Texture Scale=ScalarParameter() {defaultValue=1, group=Wrap_Ruckus, parameterName=Ruck_Twirl_Scale}) {materialFunction=&ScaleUVsByCenter}.UVs) {name=ScaledUVs, nodeColor=<rgba 1 0.375 0 1>}
  n19 = TextureObjectParameter() {
   group=Wrap_Ruckus
   parameterName=Ruckus_Mask__Text
   samplerType=SAMPLERTYPE_Masks
   sortPriority=-100
   texture=/BRCosmetics/Animation/Props/Gadget_LentilAnise/Textures/T_LentilAnisePortal_Gadget
  }
  NamedRerouteDeclaration(None=@n19) {name=RuckusMaskTexture, nodeColor=<rgba 0.0547 0 1 1>}
  n20 = NamedRerouteUsage()
  n21 = Add(A=NamedRerouteUsage(), B=Multiply(A=AppendVector(A=@n20, B=@n20)) {constB=0.005})
  n22 = Multiply(A=ScalarParameter() {defaultValue=0.1, group=Wrap_Ruckus, parameterName=Ruck_Twirl_Channel_01_RotSpeed}, B=Time())
  NamedRerouteDeclaration(None=TextureSample(
    UVs=MaterialFunctionCall(UVs=@n21, "Rotation Angle (0-1)"=@n22) {materialFunction=&CustomRotator}.Rotated Values
    Tex=NamedRerouteUsage()
   ) {samplerSource=SSM_Clamp_WorldGroupSettings, samplerType=SAMPLERTYPE_Masks}.R) {name=RedChannelTexture, nodeColor=<rgba 1 0 0 1>}
  n23 = NamedRerouteUsage()
  n24 = Add(A=NamedRerouteUsage(), B=Multiply(A=AppendVector(A=@n23, B=@n23)) {constB=0.05})
  n25 = Multiply(A=Time(), B=ScalarParameter() {defaultValue=-0.3, group=Wrap_Ruckus, parameterName=Ruck_Twirl_Channel_02_RotSpeed})
  NamedRerouteDeclaration(None=TextureSample(
    UVs=MaterialFunctionCall(UVs=@n24, "Rotation Angle (0-1)"=@n25) {materialFunction=&CustomRotator}.Rotated Values
    Tex=NamedRerouteUsage()
   ) {samplerSource=SSM_Clamp_WorldGroupSettings, samplerType=SAMPLERTYPE_Masks}.G) {name=GreenChannelTexture, nodeColor=<rgba 0 1 0 1>}
  n26 = NamedRerouteUsage()
  n27 = Add(A=NamedRerouteUsage(), B=Multiply(A=AppendVector(A=@n26, B=@n26)) {constB=0.02})
  n28 = Multiply(A=ScalarParameter() {defaultValue=-0.1, group=Wrap_Ruckus, parameterName=Ruck_Twirl_Channel_03_RotSpeed}, B=Time())
  NamedRerouteDeclaration(None=TextureSample(
    UVs=MaterialFunctionCall(UVs=@n27, "Rotation Angle (0-1)"=@n28) {materialFunction=&CustomRotator}.Rotated Values
    Tex=NamedRerouteUsage()
   ) {samplerSource=SSM_Clamp_WorldGroupSettings, samplerType=SAMPLERTYPE_Masks}.B) {name=BlueChannelTexture, nodeColor=<rgba 0 0 1 1>}
  n29 = NamedRerouteUsage()
  n30 = Multiply(A=Time(), B=ScalarParameter() {defaultValue=-0.3, group=Wrap_Ruckus, parameterName=Ruck_Twirl_Channel_04_RotSpeed})
  NamedRerouteDeclaration(None=TextureSample(
    UVs=MaterialFunctionCall(UVs=@n29, "Rotation Angle (0-1)"=@n30) {materialFunction=&CustomRotator}.Rotated Values
    Tex=NamedRerouteUsage()
   ) {samplerSource=SSM_Clamp_WorldGroupSettings, samplerType=SAMPLERTYPE_Masks}.A) {name=AlphaChannelTexture, nodeColor=<rgba 1 1 1 1>}
  n31 = VectorParameter() {group=Wrap_Ruckus, parameterName=Ruck_Fresnel_Color, sortPriority=1, defaultValue=<rgba 1 1 1 1>}
  n32 = Fresnel(
   ExponentIn=ScalarParameter() {defaultValue=8, group=Wrap_Ruckus, parameterName=Ruck_Fresnel -----, sortPriority=1}
   Normal=Transform(None=NamedRerouteUsage())
  ) {baseReflectFraction=0}
  NamedRerouteDeclaration(None=Multiply(
    A=Multiply(A=@n31.RGB, B=@n32)
    B=ScalarParameter() {defaultValue=1, group=Wrap_Ruckus, parameterName=Ruck_Fresnel Intensity, sortPriority=1}
   )) {name=Fresnel, nodeColor=<rgba 0 1 0.3672 1>}
  n33 = NamedRerouteUsage()
  n34 = LinearInterpolate(A=NamedRerouteUsage(), B=NamedRerouteUsage(), Alpha=NamedRerouteUsage())
  NamedRerouteDeclaration(None=Multiply(A=Add(A=@n33, B=@n34) {constB=0}, B=ScalarParameter() {defaultValue=0.5, group=Wrap_Ruckus, parameterName=Ruck_Emissive_Amount})) {name=EmissiveAmount, nodeColor=<rgba 1 0.0469 0 1>}
  n35 = NamedRerouteUsage()
  n36 = NamedRerouteUsage()
  n37 = Add(A=Add(A=Add(A=@n35, B=@n36), B=NamedRerouteUsage()), B=NamedRerouteUsage())
  n38 = Multiply(A=NamedRerouteUsage()) {constB=2}
  NamedRerouteDeclaration(None=Saturate(None=Add(A=@n37, B=@n38))) {name=TransparencyMask, nodeColor=<rgba 0 0.9766 1 1>}
  n39 = ScalarParameter() {defaultValue=0.41, group=Wrap_Ruckus, parameterName=Ruck_Core_SphereMask_Radius}
  n40 = ScalarParameter() {defaultValue=0.248, group=Wrap_Ruckus, parameterName=Ruck_Core_SphereMask_Hardness-----}
  NamedRerouteDeclaration(None=SphereMask(A=TextureCoordinate(), B=Constant2Vector() {g=0.5, r=0.5}, Radius=Multiply(A=@n39), Hardness=Multiply(A=@n40) {constB=8}) {attenuationRadius=1.2}) {name=SphereMaskOpacity, nodeColor=<rgba 0 1 0.6953 1>}
  NamedRerouteDeclaration(None=MaterialFunctionCall(
    A=VectorParameter() {group=Wrap_Ruckus, parameterName=LentilAniseAlphaColourA, defaultValue=<rgba 0.0764 0.1565 0.0266 1>}.RGB
    B=VectorParameter() {group=Wrap_Ruckus, parameterName=LentilAniseAlphaColourB, defaultValue=<rgba 0.1929 0.4 0.0646 1>}.RGB
    C=VectorParameter() {group=Wrap_Ruckus, parameterName=LentilAniseAlphaColourC, defaultValue=<rgba 0.1037 0.271 0 1>}.RGB
    Alpha=NamedRerouteUsage()
   ) {materialFunction=&Lerp_3Color}.Result) {name=ColorAlpha, nodeColor=<rgba 0 1 0.3672 1>}
  NamedRerouteDeclaration(None=MaterialFunctionCall(
    A=VectorParameter() {group=Wrap_Ruckus, parameterName=LentilAniseBlueColourA, defaultValue=<rgba 0 0 0 1>}.RGB
    B=VectorParameter() {group=Wrap_Ruckus, parameterName=LentilAniseBlueColourB, defaultValue=<rgba 0.3154 0.4 0.0646 1>}.RGB
    C=VectorParameter() {group=Wrap_Ruckus, parameterName=LentilAniseBlueColourC, defaultValue=<rgba 0.143 0.4635 0.0628 1>}.RGB
    Alpha=NamedRerouteUsage()
   ) {materialFunction=&Lerp_3Color}.Result) {name=ColorBlue, nodeColor=<rgba 0 0 1 1>}
  NamedRerouteDeclaration(None=MaterialFunctionCall(
    A=VectorParameter() {group=Wrap_Ruckus, parameterName=LentilAniseGreenColourA, defaultValue=<rgba 0 0 0 1>}.RGB
    B=VectorParameter() {group=Wrap_Ruckus, parameterName=LentilAniseGreenColourB, defaultValue=<rgba 0.3236 0.8229 0.0771 1>}.RGB
    C=VectorParameter() {group=Wrap_Ruckus, parameterName=LentilAniseGreenColourC, defaultValue=<rgba 0.4933 0.713 0.0966 1>}.RGB
    Alpha=NamedRerouteUsage()
   ) {materialFunction=&Lerp_3Color}.Result) {name=ColorGreen, nodeColor=<rgba 0 1 0 1>}
  NamedRerouteDeclaration(None=MaterialFunctionCall(
    A=VectorParameter() {group=Wrap_Ruckus, parameterName=LentilAniseCoreColourA, defaultValue=<rgba 0 0 0 1>}.RGB
    B=VectorParameter() {group=Wrap_Ruckus, parameterName=LentilAniseCoreColourB, defaultValue=<rgba 0.1929 0.4 0.0646 1>}.RGB
    C=VectorParameter() {group=Wrap_Ruckus, parameterName=LentilAniseCoreColourC, defaultValue=<rgba 1.3913 2 0 1>}.RGB
    Alpha=NamedRerouteUsage()
   ) {materialFunction=&Lerp_3Color}.Result) {name=EmissiveCoreColor, nodeColor=<rgba 0 0 1 1>}
  n41 = Power(
   Base=Divide(A=@n7, B=ScalarParameter() {defaultValue=0.3, group=Wrap_Ruckus, parameterName=Ruck_Core01_DistanceFalloff-----})
   Exp=ScalarParameter() {defaultValue=80, group=Wrap_Ruckus, parameterName=Ruck_Core01_Fresnel_Exp}
  )
  n42 = ScalarParameter() {defaultValue=0.9, group=Wrap_Ruckus, parameterName=Ruck_Core01_MaxBrightness}
  n43 = Add(
   A=LinearInterpolate(
    A=VectorParameter() {group=Wrap_Ruckus, parameterName=Ruck_Core01_EmColor_01, defaultValue=<rgba 1 1 1 1>}.RGB
    B=VectorParameter() {group=Wrap_Ruckus, parameterName=Ruck_Core01_EmColor_02, defaultValue=<rgba 0 0 0 1>}.RGB
    Alpha=Clamp(None=@n41, Min=@n42)
   )
   B=@n9
  )
  NamedRerouteDeclaration(None=Power(Base=Add(A=@n43, B=@n11))) {name=EmissiveCoreAlpha, nodeColor=<rgba 0 1 0.4375 1>}
  n44 = LocalPosition() {localOrigin=InstancePreSkinning}
  n45 = Multiply(A=NamedRerouteUsage(), B=ScalarParameter() {parameterName=AmountofWPO}) {constB=-10}
  NamedRerouteDeclaration(None=AppendVector(A=Constant2Vector(), B=Add(A=@n44.Z, B=@n45))) {name=WPO, nodeColor=<rgba 0.9688 0 1 1>}
  NamedRerouteDeclaration(None=@n1.Normal) {name=In_Normal, nodeColor=<rgba 0 1 0.2969 1>}
  NamedRerouteDeclaration(None=FunctionInput() {inputName=Material, inputType=FunctionInput_MaterialAttributes, sortPriority=-200}) {name=In_MaterialAttributes, nodeColor=<rgba 1 0 0.0234 1>}
  NamedRerouteUsage()
  NamedRerouteUsage()
  VertexInterpolator(VS=@n44.Z)
 ]}
