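# Material function that draws an animated, green-tinted swirling portal for the LentilAnise gadget prop.
# Packed mask channels are sampled both as rotating planar UVs and as twisted, panning polar UVs with explicit derivatives, then blended.
# These layers tint and composite with a rotating, noise-distorted background using sphere masks.
# It overrides incoming material attributes, including emissive, opacity and Z-based world position offset, then passes them through MF_TechArtOverrides.
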MaterialFunction /BRCosmetics/Animation/Props/Gadget_LentilAnise/Materials/MF_LentilAnisePortal
graph {nodes [
  n1 = NamedRerouteUsage()
  n2 = SetMaterialAttributes!(
   MaterialAttributes=GetMaterialAttributes(None=NamedRerouteUsage()) {attributeGetTypes [#1]}.MaterialAttributes
   BaseColor=VectorParameter() {group=Wrap_Ruckus, parameterName=Ruck_BaseColor-----, sortPriority=0}.RGB
   Roughness=ScalarParameter() {defaultValue=1, group=Wrap_Ruckus, parameterName=Ruck_Roughness, sortPriority=0}
   Specular=ScalarParameter() {group=Wrap_Ruckus, parameterName=Ruck_Specular, sortPriority=0}
   EmissiveColor=Multiply(A=NamedRerouteUsage(), B=ScalarParameter() {defaultValue=1, parameterName=EmissiveMultiplier})
   Opacity=@n1
   WorldPositionOffset=NamedRerouteUsage()
   OpacityMask=@n1
  ) {attributeSetTypes [
    #2
    #3
    #4
    #5
    #6
    #7
    #8
   ]}
  FunctionOutput(None=MaterialFunctionCall(In=@n2) {materialFunction=&MF_TechArtOverrides}.Result)
  n3 = TextureObjectParameter() {
   group=Wrap_Ruckus
   parameterName=RadialTexture
   samplerType=SAMPLERTYPE_Masks
   sortPriority=-100
   texture=/BRCosmetics/Animation/Props/Gadget_LentilAnise/Textures/T_LentilAnisePortalB_Gadget
  }
  NamedRerouteDeclaration(None=@n3) {name=RuckusMaskRadial, nodeColor=<rgba 0.0547 0 1 1>}
  n4 = TextureObjectParameter() {
   group=Wrap_Ruckus
   parameterName=BackgroundTexture
   sortPriority=-100
   texture=/BRCosmetics/Animation/Props/Gadget_LentilAnise/Textures/T_LentilAnisePortal_GadgetC
  }
  NamedRerouteDeclaration(None=@n4) {name=BackgroundTexture, nodeColor=<rgba 0.0547 0 1 1>}
  n5 = Multiply(A=NamedRerouteUsage()) {constB=1.5}
  NamedRerouteDeclaration(None=LinearInterpolate(A=NamedRerouteUsage(), B=NamedRerouteUsage(), Alpha=Saturate(None=@n5))) {name=BlendAlphaPlanarToRadial, nodeColor=<rgba 0.4844 0.4844 0.4844 1>}
  n6 = TextureObjectParameter() {
   group=Wrap_Ruckus
   parameterName=NoiseTexture
   samplerType=SAMPLERTYPE_Grayscale
   sortPriority=-100
   texture=&LowResBlurredNoise
  }
  NamedRerouteDeclaration(None=@n6) {name=LowResNoise, nodeColor=<rgba 0.0547 0 1 1>}
  n7 = TextureSample(
   UVs=Panner(Coordinate=TextureCoordinate(), Time=Time(), Speed=ScalarParameter() {defaultValue=0.3, group=Wrap_Ruckus, parameterName=NoisePannerSpeed})
   Tex=NamedRerouteUsage()
  ) {samplerSource=SSM_Wrap_WorldGroupSettings, samplerType=SAMPLERTYPE_Masks}
  NamedRerouteDeclaration(None=Multiply(A=@n7.R, B=@n7.R)) {name=LowResNoiseTexture, nodeColor=<rgba 1 0.08 0 1>}
  NamedRerouteDeclaration(None=NamedRerouteUsage()) {name=ColorBackgroundGradient, nodeColor=<rgba 0.9796 1 0.3647 1>}
  n8 = Time()
  n9 = ScalarParameter() {defaultValue=0.01, group=Wrap_Ruckus, parameterName=OuterEdgeRotSpeed}
  n10 = Multiply(A=NamedRerouteUsage()) {constB=2}
  n11 = Add(
   A=MaterialFunctionCall(UVs=NamedRerouteUsage(), "Rotation Angle (0-1)"=Multiply(A=@n8, B=@n9)) {materialFunction=&CustomRotator}.Rotated Values
   B=Multiply(A=Subtract(A=@n10), B=ScalarParameter() {defaultValue=0.015, desc=1, parameterName=LowResNoiseMultiplier}) {constB=0.1}
  )
  n12 = TextureSample(UVs=@n11, Tex=NamedRerouteUsage())
  NamedRerouteDeclaration(None=@n12.RGB) {name=BackgroundColorTexture, nodeColor=<rgba 1 0 0.6797 1>}
  NamedRerouteDeclaration(None=@n12.A) {name=BackgroundColorAlpha, nodeColor=<rgba 1 0 0.0703 1>}
  n13 = LinearInterpolate(A=NamedRerouteUsage(), B=NamedRerouteUsage(), Alpha=NamedRerouteUsage())
  n14 = NamedRerouteUsage()
  n15 = Saturate(None=NamedRerouteUsage())
  n16 = TextureCoordinate()
  n17 = Constant2Vector() {g=0.5, r=0.5}
  n18 = SphereMask(A=TextureCoordinate(), B=Constant2Vector() {g=0.5, r=0.5}) {attenuationRadius=0.3, hardnessPercent=10}
  n19 = LinearInterpolate(
   A=NamedRerouteUsage()
   B=LinearInterpolate(A=LinearInterpolate(A=@n13, B=@n14, Alpha=@n15), B=NamedRerouteUsage(), Alpha=NamedRerouteUsage())
   Alpha=Multiply(A=SphereMask(A=@n16, B=@n17) {attenuationRadius=0.67, hardnessPercent=0.5}, B=OneMinus(None=@n18))
  ) {constAlpha=1}
  n20 = VectorParameter() {group=Wrap_Ruckus, parameterName=LittleBlobsColor, defaultValue=<rgba 0.721 1.2 0.7125 1>}
  n21 = NamedRerouteUsage()
  n22 = Multiply(A=NamedRerouteUsage()) {constB=4}
  NamedRerouteDeclaration(None=LinearInterpolate(
    A=LinearInterpolate(A=@n19, B=@n20.RGB, Alpha=@n21)
    B=VectorParameter() {group=Wrap_Ruckus, parameterName=BackgroundOuterEdgeColor, defaultValue=<rgba 0.227 0.7454 0.1221 1>}.RGB
    Alpha=Saturate(None=@n22)
   )) {name=ColorMixed, nodeColor=<rgba 1 0 0.2812 1>}
  n23 = NamedRerouteUsage()
  n24 = Multiply(A=ScalarParameter() {defaultValue=0.033, parameterName=GreenChannelOuterSpeed}, B=Time())
  NamedRerouteDeclaration(None=TextureSample(
    UVs=MaterialFunctionCall(UVs=@n23, "Rotation Angle (0-1)"=@n24) {materialFunction=&CustomRotator}.Rotated Values
    Tex=NamedRerouteUsage()
   ) {samplerSource=SSM_Clamp_WorldGroupSettings, samplerType=SAMPLERTYPE_Masks}.B) {name=GreenChannelTexture, nodeColor=<rgba 0 1 0 1>}
  NamedRerouteDeclaration(None=SphereMask(A=TextureCoordinate(), B=Constant2Vector() {g=0.5, r=0.5}) {attenuationRadius=0.5, hardnessPercent=0.5}) {name=SphereMaskBlend, nodeColor=<rgba 0 0.3437 1 1>}
  n25 = SphereMask(A=TextureCoordinate(), B=Constant2Vector() {g=0.5, r=0.5}) {attenuationRadius=0.5, hardnessPercent=0.5}
  n26 = MaterialFunctionCall!(
   UV=MaterialFunctionCall(UVs=TextureCoordinate(), Texture Scale=Constant2Vector() {g=1, r=1}) {materialFunction=&ScaleUVsByCenter}.UVs
   Twist Center=Multiply(
    A=NamedRerouteDeclaration(None=@n25) {name=SphereMaskBlend_3, nodeColor=<rgba 0 0.3438 1 1>}
    B=ScalarParameter() {defaultValue=1.2, parameterName=AmountOfTwistBlueChannel}
   ) {constB=2}
   Twist Edges=Constant() {r=0.25}
  ) {materialFunction=&/MF_PolarUV}
  n27 = Time()
  n28 = VectorParameter() {parameterName=BlueChannelRadialSpeed, defaultValue=<rgba 0.154 0.068 0 1>}
  n29 = Add(A=@n26.Polar UVs, B=ComponentMask(None=Multiply(A=@n27, B=@n28.RGB) {constB=0.1}) {g=true, r=true})
  n30 = NamedRerouteUsage()
  NamedRerouteDeclaration(None=Power(Base=TextureSample!(UVs=@n29, Tex=@n30, "DDX(UVs)"=@n26.DDX, "DDY(UVs)"=@n26.DDX, Apply View MipBias) {mipValueMode=TMVM_Derivative, samplerSource=SSM_Wrap_WorldGroupSettings, samplerType=SAMPLERTYPE_Masks}.B) {constExponent=4}) {name=BlueChannelTextureRadial, nodeColor=<rgba 0 0 1 1>}
  NamedRerouteDeclaration(None=FunctionInput() {inputName=Material, inputType=FunctionInput_MaterialAttributes, sortPriority=-200}) {name=In_MaterialAttributes, nodeColor=<rgba 1 0 0.0234 1>}
  n31 = NamedRerouteUsage()
  NamedRerouteDeclaration(None=LinearInterpolate(A=NamedRerouteUsage(), B=NamedRerouteUsage(), Alpha=Multiply(A=@n31) {constB=1.5})) {name=BlendGreenPlanarToRadial, nodeColor=<rgba 0 1 0 1>}
  NamedRerouteDeclaration(None=LinearInterpolate(
    A=NamedRerouteUsage()
    B=VectorParameter() {group=Wrap_Ruckus, parameterName=LentilAniseGreenColourC, defaultValue=<rgba 0.3548 0.651 0.0736 1>}.RGB
    Alpha=NamedRerouteUsage()
   )) {name=ColorGreen, nodeColor=<rgba 0 1 0.9297 1>}
  n32 = TextureObjectParameter() {
   group=Wrap_Ruckus
   parameterName=PlanarTexture
   samplerType=SAMPLERTYPE_Masks
   sortPriority=-100
   texture=/BRCosmetics/Animation/Props/Gadget_LentilAnise/Textures/T_LentilAnisePortal_Gadget
  }
  NamedRerouteDeclaration(None=@n32) {name=RuckusMaskTexture, nodeColor=<rgba 0.0547 0 1 1>}
  n33 = NamedRerouteUsage()
  n34 = Multiply(A=ScalarParameter() {defaultValue=0.05, group=Wrap_Ruckus, parameterName=Ruck_Twirl_Channel_01_RotSpeed}, B=Time())
  NamedRerouteDeclaration(None=TextureSample!(
    UVs=MaterialFunctionCall(UVs=@n33, "Rotation Angle (0-1)"=@n34) {materialFunction=&CustomRotator}.Rotated Values
    Tex=NamedRerouteUsage()
    Apply View MipBias
   ) {samplerSource=SSM_Clamp_WorldGroupSettings, samplerType=SAMPLERTYPE_Masks}.R) {name=WhiteDots, nodeColor=<rgba 1 1 1 1>}
  n35 = SphereMask(A=TextureCoordinate(), B=Constant2Vector() {g=0.5, r=0.5}) {attenuationRadius=0.5, hardnessPercent=12}
  n36 = MaterialFunctionCall(
   UV=MaterialFunctionCall(UVs=TextureCoordinate(), Texture Scale=Constant2Vector() {g=1, r=1}) {materialFunction=&ScaleUVsByCenter}.UVs
   Twist Center=Multiply(
    A=NamedRerouteDeclaration(None=@n35) {name=SphereMaskBlend_4, nodeColor=<rgba 0 0.3438 1 1>}
    B=ScalarParameter() {defaultValue=1, parameterName=AmountOfTwistWhiteBitsChannel}
   ) {constB=2}
   Twist Edges=Constant() {r=0.2}
  ) {materialFunction=&MF_PolarUV}
  n37 = Time()
  n38 = VectorParameter() {parameterName=WhiteBitsChannelRadialSpeed, defaultValue=<rgba 0.074 0.052 0 1>}
  n39 = Add(A=@n36.Polar UVs, B=ComponentMask(None=Multiply(A=@n37, B=@n38.RGB) {constB=0.1}) {g=true, r=true})
  n40 = NamedRerouteUsage()
  NamedRerouteDeclaration(None=Multiply(A=TextureSample!(UVs=@n39, Tex=@n40, "DDX(UVs)"=@n36.DDX, "DDY(UVs)"=@n36.DDX, Apply View MipBias) {mipValueMode=TMVM_Derivative, samplerSource=SSM_Wrap_WorldGroupSettings, samplerType=SAMPLERTYPE_Masks}.A)) {name=WhiteBitsChannelTextureRadial, nodeColor=<rgba 1 1 1 1>}
  n41 = NamedRerouteUsage()
  n42 = Multiply(A=ScalarParameter() {defaultValue=0.033, parameterName=BlueChannelOuterSpeed}, B=Time())
  NamedRerouteDeclaration(None=TextureSample!(
    UVs=MaterialFunctionCall(UVs=@n41, "Rotation Angle (0-1)"=@n42) {materialFunction=&CustomRotator}.Rotated Values
    Tex=NamedRerouteUsage()
    Apply View MipBias
   ) {samplerSource=SSM_Clamp_WorldGroupSettings, samplerType=SAMPLERTYPE_Masks}.B) {name=BlueChannelTexture, nodeColor=<rgba 0 0 1 1>}
  n43 = NamedRerouteUsage()
  n44 = Multiply(A=Time(), B=ScalarParameter() {defaultValue=0.15, group=Wrap_Ruckus, parameterName=Ruck_Twirl_Channel_04_RotSpeed})
  NamedRerouteDeclaration(None=TextureSample(
    UVs=MaterialFunctionCall(UVs=@n43, "Rotation Angle (0-1)"=@n44) {materialFunction=&CustomRotator}.Rotated Values
    Tex=NamedRerouteUsage()
   ) {samplerSource=SSM_Clamp_WorldGroupSettings, samplerType=SAMPLERTYPE_Masks}.A) {name=AlphaChannelTexture, nodeColor=<rgba 1 1 1 1>}
  NamedRerouteDeclaration(None=Multiply(A=NamedRerouteUsage()) {constB=12}) {name=TransparencyMask, nodeColor=<rgba 0 0.9766 1 1>}
  NamedRerouteDeclaration(None=LinearInterpolate(
    A=NamedRerouteUsage()
    B=VectorParameter() {group=Wrap_Ruckus, parameterName=LentilAniseAlphaColourC, defaultValue=<rgba 0.1568 0.8879 0.0555 1>}.RGB
    Alpha=NamedRerouteUsage()
   )) {name=ColorAlpha, nodeColor=<rgba 0.3478 0.3478 0.3672 1>}
  NamedRerouteDeclaration(None=LinearInterpolate(
    A=NamedRerouteUsage()
    B=VectorParameter() {group=Wrap_Ruckus, parameterName=LentilAniseBlueColourC, defaultValue=<rgba 0.1821 0.6615 0.0896 1>}.RGB
    Alpha=NamedRerouteUsage()
   )) {name=ColorBlue, nodeColor=<rgba 0 0 1 1>}
  n45 = NamedRerouteUsage()
  NamedRerouteDeclaration(None=LinearInterpolate(A=NamedRerouteUsage(), B=NamedRerouteUsage(), Alpha=Multiply(A=@n45))) {name=BlendWhiteBitsPlanarToRadial, nodeColor=<rgba 1 1 1 1>}
  NamedRerouteDeclaration(None=LinearInterpolate(
    A=NamedRerouteUsage()
    B=VectorParameter() {group=Wrap_Ruckus, parameterName=LentilAniseInnerWhiteBits, defaultValue=<rgba 1.5806 2 1.1826 1>}.RGB
    Alpha=NamedRerouteUsage()
   )) {name=ColorInnerWhiteBits, nodeColor=<rgba 1 0.1721 0.0476 1>}
  NamedRerouteDeclaration(None=MaterialFunctionCall(UVs=TextureCoordinate(), Texture Scale=ScalarParameter() {defaultValue=1.33, group=Wrap_Ruckus, parameterName=UVScaleEvenBigger}) {materialFunction=&ScaleUVsByCenter}.UVs) {name=ScaledUVsEvenBigger, nodeColor=<rgba 1 0.375 0 1>}
  NamedRerouteDeclaration(None=TextureSample(UVs=@n11, Tex=NamedRerouteUsage()) {samplerSource=SSM_Clamp_WorldGroupSettings}.R) {name=BackgroundOuterEdge, nodeColor=<rgba 0.0171 1 0 1>}
  n46 = LocalPosition() {localOrigin=InstancePreSkinning}
  n47 = Multiply(A=NamedRerouteUsage(), B=ScalarParameter() {parameterName=AmountofWPO}) {constB=-10}
  NamedRerouteDeclaration(None=AppendVector(A=Constant2Vector(), B=Add(A=@n46.Z, B=@n47))) {name=WPO, nodeColor=<rgba 0.9688 0 1 1>}
  n48 = TextureCoordinate()
  n49 = Constant2Vector() {g=0.5, r=0.5}
  n50 = Multiply() {constA=0.333}
  n51 = Constant() {r=0.2}
  NamedRerouteDeclaration(None=Power(Base=SphereMask(A=@n48, B=@n49, Radius=@n50, Hardness=@n51) {attenuationRadius=1.2}) {constExponent=1}) {name=SphereMaskDistort, nodeColor=<rgba 0 1 0.6953 1>}
  NamedRerouteDeclaration(None=MaterialFunctionCall(UVs=TextureCoordinate(), Texture Scale=ScalarParameter() {defaultValue=1, group=Wrap_Ruckus, parameterName=UVScaleBig}) {materialFunction=&ScaleUVsByCenter}.UVs) {name=ScaledUVsBig, nodeColor=<rgba 1 0.375 0 1>}
  n52 = SphereMask(A=TextureCoordinate(), B=Constant2Vector() {g=0.5, r=0.5}) {attenuationRadius=0.5, hardnessPercent=0.5}
  n53 = MaterialFunctionCall(
   UV=MaterialFunctionCall(UVs=TextureCoordinate(), Texture Scale=Constant2Vector() {g=1, r=1}) {materialFunction=&ScaleUVsByCenter}.UVs
   Twist Center=Multiply(
    A=NamedRerouteDeclaration(None=@n52) {name=SphereMaskBlend_1, nodeColor=<rgba 0 0.3438 1 1>}
    B=ScalarParameter() {defaultValue=1.4, parameterName=AmountOfTwistBlueChannel}
   ) {constB=2}
   Twist Edges=Constant() {r=0.5}
  ) {materialFunction=&MF_PolarUV}
  n54 = Time()
  n55 = VectorParameter() {parameterName=GreenChannelRadialSpeed, defaultValue=<rgba 0.1 0.05 0 1>}
  n56 = Add(A=@n53.Polar UVs, B=ComponentMask(None=Multiply(A=@n54, B=@n55.RGB) {constB=0.1}) {g=true, r=true})
  n57 = NamedRerouteUsage()
  NamedRerouteDeclaration(None=Power(Base=TextureSample!(UVs=@n56, Tex=@n57, "DDX(UVs)"=@n53.DDX, "DDY(UVs)"=@n53.DDX, Apply View MipBias) {mipValueMode=TMVM_Derivative, samplerSource=SSM_Wrap_WorldGroupSettings, samplerType=SAMPLERTYPE_Masks}.G) {constExponent=4}) {name=GreenChannelTextureRadial, nodeColor=<rgba 0 1 0 1>}
  n58 = NamedRerouteUsage()
  NamedRerouteDeclaration(None=LinearInterpolate(A=NamedRerouteUsage(), B=NamedRerouteUsage(), Alpha=Multiply(A=@n58) {constB=1.5})) {name=BlendBluePlanarToRadial, nodeColor=<rgba 0 0.2947 1 1>}
  n59 = SphereMask(A=TextureCoordinate(), B=Constant2Vector() {g=0.5, r=0.5}) {attenuationRadius=1, hardnessPercent=0.5}
  n60 = MaterialFunctionCall(
   UV=MaterialFunctionCall(UVs=TextureCoordinate(), Texture Scale=Constant2Vector() {g=1, r=1}) {materialFunction=&ScaleUVsByCenter}.UVs
   Twist Center=Multiply(
    A=NamedRerouteDeclaration(None=@n59) {name=SphereMaskBlend_2, nodeColor=<rgba 0 0.3438 1 1>}
    B=ScalarParameter() {defaultValue=1, parameterName=AmountOfTwistAlphaChannel}
   ) {constB=2}
   Twist Edges=Constant() {r=0.3}
  ) {materialFunction=&MF_PolarUV}
  n61 = Multiply(A=Time(), B=VectorParameter() {parameterName=AlphaChannelRadialSpeed, defaultValue=<rgba 0.045 0.05 0 1>}.RGB) {constB=0.1}
  n62 = TextureSample(UVs=Add(A=@n60.Polar UVs, B=ComponentMask(None=@n61) {g=true, r=true}), Tex=NamedRerouteUsage(), "DDX(UVs)"=@n60.DDX, "DDY(UVs)"=@n60.DDX) {mipValueMode=TMVM_Derivative, samplerSource=SSM_Wrap_WorldGroupSettings, samplerType=SAMPLERTYPE_Masks}
  n63 = SphereMask(A=TextureCoordinate(), B=Constant2Vector() {g=0.5, r=0.5}) {attenuationRadius=1, hardnessPercent=0.5}
  n64 = MaterialFunctionCall(
   UV=MaterialFunctionCall(UVs=TextureCoordinate(), Texture Scale=Constant2Vector() {g=1, r=1}) {materialFunction=&ScaleUVsByCenter}.UVs
   Twist Center=Multiply(
    A=NamedRerouteDeclaration(None=@n63) {name=SphereMaskBlend_5, nodeColor=<rgba 0 0.3438 1 1>}
    B=ScalarParameter() {defaultValue=1, parameterName=AmountOfTwistAlphaChannelB}
   ) {constB=2}
   Twist Edges=Constant() {r=0.3}
  ) {materialFunction=&MF_PolarUV}
  n65 = Time()
  n66 = VectorParameter() {parameterName=AlphaChannelRadialSpeedB, defaultValue=<rgba 0.05 0.1 0 1>}
  n67 = Add(A=@n64.Polar UVs, B=ComponentMask(None=Multiply(A=@n65, B=@n66.RGB) {constB=0.1}) {g=true, r=true})
  n68 = NamedRerouteUsage()
  n69 = Multiply(A=Power(Base=TextureSample(UVs=@n67, Tex=@n68, "DDX(UVs)"=@n64.DDX, "DDY(UVs)"=@n64.DDX) {mipValueMode=TMVM_Derivative, samplerSource=SSM_Wrap_WorldGroupSettings, samplerType=SAMPLERTYPE_Masks}.G)) {constB=4}
  NamedRerouteDeclaration(None=Add(
    A=Power(Base=@n62.G) {constExponent=4}
    B=NamedRerouteDeclaration(None=@n69) {name=AlphaChannelTextureRadial_1, nodeColor=<rgba 0.4224 0.4224 0.4688 1>}
   )) {name=AlphaChannelTextureRadial, nodeColor=<rgba 0.4224 0.4224 0.4688 1>}
 ]}
