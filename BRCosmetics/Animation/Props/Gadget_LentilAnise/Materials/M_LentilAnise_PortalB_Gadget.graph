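# Masked, unlit, two-sided material for the Portal B gadget of the LentilAnise cosmetic prop, flagged for Niagara sprites and pixel animation.
# It holds almost no logic of its own: a GetMaterialAttributes node reads the output of the MF_LentilAnisePortal material function.
# The portal's look is therefore defined in that function rather than in this graph.
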Material /BRCosmetics/Animation/Props/Gadget_LentilAnise/Materials/M_LentilAnise_PortalB_Gadget
graph {
 properties {
  bAutomaticallySetUsageInEditor=false
  bHasPixelAnimation=true
  bUsedWithNiagaraSprites=true
  blendMode=BLEND_Masked
  shadingModel=MSM_Unlit
  twoSided=true
  usedShadingModels=Unlit
 }
 nodes [GetMaterialAttributes(None=MaterialFunctionCall!(Material) {materialFunction=&/MF_LentilAnisePortal}.Result) {attributeGetTypes [#1, #2]}]
}
