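# Physics asset for the LentilAnise gadget prop mesh, defining its collision and ragdoll-style physics setup.
# Four capsule bodies cover the container, core, and left and right door bones.
# The core and both doors each attach to the container through constraints with matching swing and twist limits, forming a star around the container as hub.
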PhysicsAsset /BRCosmetics/Animation/Props/Gadget_LentilAnise/Meshes/LentilAnise_Gadget_PhysicsAsset
bodies {
 JNT_Container [{shapeType=Capsule, length=37.1472, radius=26.3552, center=<0.5981 0.0128 -9.2704>, rotation=<rot 0 179.9526 -89.9207>}]
 JNT_Core [{shapeType=Capsule, length=44.2174, radius=9.7831, rotation=<rot 0 -90 -90>}]
 JNT_L_Door [{shapeType=Capsule, length=25.6745, radius=8.6746, center=<5.0425 0 0>, rotation=<rot 0 -90 0>}]
 JNT_R_Door [{shapeType=Capsule, length=25.6745, radius=8.6886, center=<5.0563 0 0>, rotation=<rot 0 -90 0>}]
}
constraints [
 {bone1Name=JNT_Core, bone2Name=JNT_Container, swing1LimitDegrees=45, swing2LimitDegrees=45, twistLimitDegrees=45}
 {bone1Name=JNT_L_Door, bone2Name=JNT_Container, swing1LimitDegrees=45, swing2LimitDegrees=45, twistLimitDegrees=45}
 {bone1Name=JNT_R_Door, bone2Name=JNT_Container, swing1LimitDegrees=45, swing2LimitDegrees=45, twistLimitDegrees=45}
]
